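; BP_Ancients_WallSconce_A is an Ancients-themed wall sconce prop with a looping fire sound and time-of-day lighting.
; The construction script initializes its TODLightingComponent through RunConstruct, passing the Test Day Phase variable.
; On death, it checks that the fire loop audio component is valid, then quickly fades it out.
; BeginPlay, Tick and overlap events are empty.
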(event EventBeginPlay)

(event Collision|EventActorBeginOverlap
  (params (OtherActor :type Actor%20Object%20Reference)))

(event EventTick
  (params (DeltaSeconds :type Float%20%28single-precision%29)))

(event EventOnDeathPlayEffects
  (params (Damage :type Float%20%28single-precision%29) (DamageTags :type Gameplay%20Tag%20Container%20Structure%20%28by%20ref%29) (Momentum :type Vector :default "0, 0, 0") (HitInfo :type Hit%20Result%20Structure%20%28by%20ref%29) (InstigatedBy :type Fort%20Pawn%20Object%20Reference) (DamageCauser :type Actor%20Object%20Reference) (EffectContext :type Gameplay%20Effect%20Context%20Handle%20Structure))
  (bind _spot_wallsconce_fire_loop_cue (Variables|Default|GetSpot_WallSconce_Fire_Loop_Cue))
  ;[1 1 1] On Death Stop Ambient Fire Loop
  (Utilities|IsValid _spot_wallsconce_fire_loop_cue
    (:Is%20Valid
      (if (Utilities|IsValid _spot_wallsconce_fire_loop_cue :!class /Script/Engine.KismetSystemLibrary)
        (Audio|Components|Audio|FadeOut _spot_wallsconce_fire_loop_cue 0.2 0.0 "Linear" :!class /Script/Engine.AudioComponent)))))

(fn ConstructionScript
  (Class|TODLightingComponent|RunConstruct (Variables|Default|GetTODLightingComponent) (Variables|Default|GetTest%20Day%20Phase)) ; Initialize the Time of Day component.  
)
